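; Rift item portal that teleports its owner and overlapping players, vehicles and AI pawns to a configured height at its location until a timer closes it.
; TryRiftTeleport branches by target type, with server authority, suppressing fall damage and sending rift trigger events; ForceRiftTeleport skips the open check.
; A growing, timer-reset offset separates rapid successive teleports.
; Timelines drive opening, proximity widening and closing audio; quest progress is awarded based on a Creative-mode check.
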(event EventBeginPlay
  ;[1 1 1] EVENT--> BeginPlay
  (|CallParentFunction)
  (Variables|RiftPortal|SetPortalOpen true)
  (CallFunction|InitTeleportLocation)
  (CallFunction|PlayCosmeticFX)
  (Utilities|Time|SetTimerbyEvent (EventDispatchers|CreateEvent :!event Custom|DestroyRift) (ScalableFloat|GetValueatLevel (Variables|RiftPortal|GetActiveDuration)))
  ;[1 1 1 0.5] FORT-1084529; Make sure the player that activates ties to teleport
  (CallFunction|TryRiftTeleport (Actor|GetOwner))
  ;[1 1 1] Check if In Creative Mode or not for Quest
  (bind _asfort_game_state_athena (Utilities|Casting|CastToFortGameStateAthena (Game|GetGameState)))
  (if (Variables|/Script/FortniteGame.FortGameStateAthena|GetbPlaylistDataIsLoaded _asfort_game_state_athena)
    (bind _cancompletequest (Class|AthenaQuestFunctionLibrary|AthenaCheckCreativeMode _asfort_game_state_athena))
    (if _cancompletequest
      (Variables|QuestStuff|SetCreativeMode false)
      (else
        (Variables|QuestStuff|SetCreativeMode true)))
    (else
      (Playlist|BindEventtoOnCurrentPlaylistLoaded _asfort_game_state_athena (delegate Custom|PlaylistLoaded)))))

(event Custom|PlaylistLoaded
  (params (PlaylistName :type Name) (PlaylistContextTags :type Gameplay%20Tag%20Container%20Structure%20%28by%20ref%29))
  ;[1 1 1] Check if In Creative Mode or not for Quest
  (bind _cancompletequest (Class|AthenaQuestFunctionLibrary|AthenaCheckCreativeMode (stale EventBeginPlay _asfort_game_state_athena)))
  (if _cancompletequest
    (Variables|QuestStuff|SetCreativeMode false)
    (else
      (Variables|QuestStuff|SetCreativeMode true))))

(event Custom|PlayTeleportFX
  (params (PlayerLocation :type Vector) (RequestingPlayerState :type Player%20State%20Object%20Reference))
  (bind _enteredrift (Variables|Sound|GetEnteredRift))
  ;[1 1 1] Spawn teleport visuals and audio for clients
  (Audio|PlaySoundatLocation _enteredrift PlayerLocation)
  (SoundIndicator|ShowSoundIndicator (Variables|Default|GetSoundIndicator) PlayerLocation _enteredrift 1.0 RequestingPlayerState)
  (Utilities|FlowControl|Delay 0.5)
  (Effects|Components|ParticleSystem|SpawnFXAtLocation (Variables|VFX|GetEnterRift_NS) PlayerLocation)
  (Utilities|FlowControl|Delay 1.0)
  (Audio|SpawnSoundatLocation (Variables|Sound|GetPlayerExitSkyCrack) (Variables|RiftPortal|GetTeleportLocation)))

(event Custom|PlayCosmeticFX
  (bind _sm_rift (Variables|Default|GetSM_Rift))
  ;[1 1 1] Toggle FX and animate shards after opening
  (Utilities|FlowControl|Sequence
    (:then_0
      (Components|Activation|SetActive (Variables|Default|GetNS_Rift_Idle_Loop) false false)
      (Components|Activation|SetActive (Variables|Default|GetNS_Rift_PortalClosing) true false)
      (bind (_direction _open _glow) (Timeline|FragmentMovement Play)
        (:Update
          (Rendering|Material|SetScalarParameterValueonMaterials _sm_rift "OpenAdd" _open)
          (Rendering|Material|SetScalarParameterValueonMaterials _sm_rift "Emissive" _glow))))
    (:then_1
      (Utilities|FlowControl|Delay 1.0)
      (bind _returnvalue (Effects|Components|ParticleSystem|SpawnFXAtLocation 0 (+ (Variables|RiftPortal|GetTeleportLocation) "0, 0,200.000000") (Math|Rotator|MakeRotator 0.0 0.0 (Math|Random|RandomFloatinRange -1.8e+02 1.8e+02)) "1.000000,1.000000,1.000000" true "None" true))
      (LOD|SetMaxDrawDistance _returnvalue 0.0))))

(event OnComponentBeginOverlap(OverlapCapsule)
  (params (OverlappedComponent :type Primitive%20Component%20Object%20Reference) (OtherActor :type Actor%20Object%20Reference) (OtherComp :type Primitive%20Component%20Object%20Reference) (OtherBodyIndex :type Integer) (bFromSweep :type Boolean) (SweepResult :type Hit%20Result%20Structure%20%28by%20ref%29))
  ;[1 1 1] Teleport Player and/or Vehicle
  (if (not (or (CallFunction|IsIgnoredActor OtherActor) (Utilities|Set|ContainsItem (Variables|Default|GetBeganOverlapList) OtherActor)))
    (Utilities|Set|Add (Variables|Default|GetBeganOverlapList) OtherActor)
    (CallFunction|TryRiftTeleport OtherActor)))

(event Custom|TryRiftTeleport
  (params (TargetActor :type Actor%20Object%20Reference))
  (bind _vehicleinterface (Variables|Default|GetVehicleInterface))
  (bind _returnvalue (Vehicle|GetDriver _vehicleinterface))
  (bind _playerpawn (Variables|RiftPortal|GetPlayerPawn))
  ;[1 1 1] Teleport Player and/or Vehicle
  (Variables|RiftPortal|SetOverlapActor TargetActor)
  (if (Variables|RiftPortal|GetPortalOpen)
    (Variables|RiftPortal|SetActorRotation (Transformation|GetActorRotation (Variables|RiftPortal|GetOverlapActor)))
    (if (Networking|HasAuthority)
      (Utilities|IsValid (Variables|RiftPortal|GetOverlapActor)
        (:Is%20Valid
          (bind _asfort_player_pawn (Utilities|Casting|CastToFortPlayerPawn (Variables|RiftPortal|GetOverlapActor))
            (:then
              (if (not (Variables|/Script/FortniteGame.FortPawn|GetbIsDBNO _asfort_player_pawn))
                (Variables|RiftPortal|SetPlayerPawn _asfort_player_pawn)
                (if (not (Class|FortPawn|IsInsideHidingProp _asfort_player_pawn true))
                  (CallFunction|EjectfromTurret _asfort_player_pawn)
                  (CallFunction|PlayTeleportFX (Transformation|GetActorLocation _asfort_player_pawn) (Variables|/Script/Engine.Pawn|GetPlayerState _asfort_player_pawn))
                  (if (not (Vehicle|IsinVehicle _asfort_player_pawn))
                    (Networking|SwitchHasAuthority
                      (:Authority
                        (CallFunction|CancelPlayerPawnAbilities)
                        (RemoteControlPawn|ForceDetachfromRemoteControlledPawn _asfort_player_pawn)
                        (Variables|/Script/FortniteGame.FortPawn|SetbIgnoreNextFallingDamage true _asfort_player_pawn) ; Have to do this because of the delay. Otherwise player can die from fall damage if they fall from high distance.
                        (bind _newparam (CallFunction|IsFlying))
                        (if (not _newparam)
                          (Utilities|FlowControl|Sequence
                            (:then_0
                              (Parachute|EndSkydiving _asfort_player_pawn) ; End Skydiving if player was previously
                              (Environment|ApplyGlobalEnvironmentGameplayEffecttoActor _asfort_player_pawn "/Game/Athena/Items/Consumables/RiftItem/GE_RiftPortal_Teleported.GE_RiftPortal_Teleported_C")
                              (CallFunction|TeleportPlayerandSendEvent _asfort_player_pawn) ; Collapsed to function, as the teleport could \ntrigger the overlap of another actor that \nwould null out the actor we are teleporting \nbefore the event was applied
                            )
                            (:then_1
                              ;[1 1 1] SERVER: Check to Award Single Use Quest Progression if not in Creative Mode
                              (if (not (Variables|QuestStuff|GetCreativeMode))
                                (Networking|SwitchHasAuthority
                                  (:Authority
                                    (Utilities|IsValid _playerpawn
                                      (:Is%20Valid
                                        (CallFunction|UseRiftQuestUpdate (Pawn|GetController _playerpawn)))
                                      (:Is%20Not%20Valid
                                        (Utilities|IsValid (Utilities|ToObject(Interface) _vehicleinterface)
                                          (:Is%20Valid
                                            (Utilities|IsValid _returnvalue
                                              (:Is%20Valid
                                                (CallFunction|UseRiftQuestUpdate (Pawn|GetController _returnvalue)))
                                              (:Is%20Not%20Valid
                                                _))
                                            (for _array_element (Vehicle|GetAllPassengers _vehicleinterface)
                                              (Utilities|IsValid _array_element
                                                (:Is%20Valid
                                                  (CallFunction|UseRiftQuestUpdate (Pawn|GetController _array_element)))))))))))))))))))))
            (:CastFailed
              (bind _asfort_vehicle_interface (Utilities|Casting|CastToFortVehicleInterface (Variables|RiftPortal|GetOverlapActor))
                (:then
                  (bind _output_get (Variables|Default|SetVehicleInterface _asfort_vehicle_interface))
                  (if (== (Utilities|Array|Length (Vehicle|GetAllPassengers _output_get)) 0)
                    (bind _newparam_2 (CallFunction|IsFlying))
                    (if _newparam_2
                      _
                      (else
                        _)))
                  (CallFunction|PlayTeleportFX (Class|FortVehicleInterface|GetOwnerLocation (Variables|Default|GetVehicleInterface)) (select (Utilities|IsArrayNonEmpty (Vehicle|GetAllPassengers (Variables|Default|GetVehicleInterface))) (Variables|/Script/Engine.Pawn|GetPlayerState (Utilities|Array|Get(acopy) (Vehicle|GetAllPassengers (Variables|Default|GetVehicleInterface)))) 0))
                  (CallFunction|VehicleTeleportEvent)
                  (Development|PrintString "
============
SUCCESS
is a vehicle
============
"))
                (:CastFailed
                  (bind _asfort_aipawn (Utilities|Casting|CastToFortAIPawn (Variables|RiftPortal|GetOverlapActor)))
                  (Utilities|IsValid (.AttachedPlayerPawn (CallFunction|IsPlayerAttachedtoAIPawn _asfort_aipawn))
                    (:Is%20Valid
                      (CallFunction|TryRiftTeleport (.AttachedPlayerPawn (CallFunction|IsPlayerAttachedtoAIPawn _asfort_aipawn))) ; Teleport the Rider first
                    )
                    (:Is%20Not%20Valid
                      _))
                  (CallFunction|PlayTeleportFX (Transformation|GetActorLocation) (Variables|/Script/Engine.Pawn|GetPlayerState _asfort_aipawn))
                  (Environment|ApplyGlobalEnvironmentGameplayEffecttoActor 0 "/Game/Athena/Items/Consumables/RiftItem/GE_RiftPortal_Teleported.GE_RiftPortal_Teleported_C")
                  (AI|TeleportAIPawn _asfort_aipawn (Math|Vector|vector+vector (Variables|RiftPortal|GetTeleportLocation) (1e+02 1e+02 0.0)) "0, 0, 0" true)
                  (Variables|/Script/FortniteGame.FortPawn|SetbIgnoreNextFallingDamage true _asfort_aipawn)
                  (Ability|SendGameplayEventtoActor _asfort_aipawn "(TagName=\"Abilities.Generic.Rift.Trigger\")"))))))))))

(event Custom|ForceRiftTeleport
  (params (Target%20Actor :type Actor%20Object%20Reference))
  (bind _vehicleinterface (Variables|Default|GetVehicleInterface))
  (bind _returnvalue (Vehicle|GetDriver _vehicleinterface))
  (bind _playerpawn (Variables|RiftPortal|GetPlayerPawn))
  ;[1 1 1] Teleport Player and/or Vehicle
  ;;[1 1 1] Ensures we rift a target
  (Variables|RiftPortal|SetOverlapActor Target%20Actor)
  (Variables|RiftPortal|SetActorRotation (Transformation|GetActorRotation (Variables|RiftPortal|GetOverlapActor)))
  (if (Networking|HasAuthority)
    (Utilities|IsValid (Variables|RiftPortal|GetOverlapActor)
      (:Is%20Valid
        (bind _asfort_player_pawn (Utilities|Casting|CastToFortPlayerPawn (Variables|RiftPortal|GetOverlapActor))
          (:then
            (if (not (Variables|/Script/FortniteGame.FortPawn|GetbIsDBNO _asfort_player_pawn))
              (Variables|RiftPortal|SetPlayerPawn _asfort_player_pawn)
              (if (not (Class|FortPawn|IsInsideHidingProp _asfort_player_pawn true))
                (CallFunction|EjectfromTurret _asfort_player_pawn)
                (CallFunction|PlayTeleportFX (Transformation|GetActorLocation _asfort_player_pawn) (Variables|/Script/Engine.Pawn|GetPlayerState _asfort_player_pawn))
                (if (not (Vehicle|IsinVehicle _asfort_player_pawn))
                  (Networking|SwitchHasAuthority
                    (:Authority
                      (CallFunction|CancelPlayerPawnAbilities)
                      (RemoteControlPawn|ForceDetachfromRemoteControlledPawn _asfort_player_pawn)
                      (Variables|/Script/FortniteGame.FortPawn|SetbIgnoreNextFallingDamage true _asfort_player_pawn) ; Have to do this because of the delay. Otherwise player can die from fall damage if they fall from high distance.
                      (bind _newparam (CallFunction|IsFlying))
                      (if (not _newparam)
                        (Utilities|FlowControl|Sequence
                          (:then_0
                            (Parachute|EndSkydiving _asfort_player_pawn) ; End Skydiving if player was previously
                            (Environment|ApplyGlobalEnvironmentGameplayEffecttoActor _asfort_player_pawn "/Game/Athena/Items/Consumables/RiftItem/GE_RiftPortal_Teleported.GE_RiftPortal_Teleported_C")
                            (CallFunction|TeleportPlayerandSendEvent _asfort_player_pawn) ; Collapsed to function, as the teleport could \ntrigger the overlap of another actor that \nwould null out the actor we are teleporting \nbefore the event was applied
                          )
                          (:then_1
                            ;[1 1 1] SERVER: Check to Award Single Use Quest Progression if not in Creative Mode
                            (if (not (Variables|QuestStuff|GetCreativeMode))
                              (Networking|SwitchHasAuthority
                                (:Authority
                                  (Utilities|IsValid _playerpawn
                                    (:Is%20Valid
                                      (CallFunction|UseRiftQuestUpdate (Pawn|GetController _playerpawn)))
                                    (:Is%20Not%20Valid
                                      (Utilities|IsValid (Utilities|ToObject(Interface) _vehicleinterface)
                                        (:Is%20Valid
                                          (Utilities|IsValid _returnvalue
                                            (:Is%20Valid
                                              (CallFunction|UseRiftQuestUpdate (Pawn|GetController _returnvalue)))
                                            (:Is%20Not%20Valid
                                              _))
                                          (for _array_element (Vehicle|GetAllPassengers _vehicleinterface)
                                            (Utilities|IsValid _array_element
                                              (:Is%20Valid
                                                (CallFunction|UseRiftQuestUpdate (Pawn|GetController _array_element)))))))))))))))))))))
          (:CastFailed
            (bind _asfort_vehicle_interface (Utilities|Casting|CastToFortVehicleInterface (Variables|RiftPortal|GetOverlapActor))
              (:then
                (bind _output_get (Variables|Default|SetVehicleInterface _asfort_vehicle_interface))
                (if (== (Utilities|Array|Length (Vehicle|GetAllPassengers _output_get)) 0)
                  (bind _newparam_2 (CallFunction|IsFlying))
                  (if _newparam_2
                    _
                    (else
                      _)))
                (CallFunction|PlayTeleportFX (Class|FortVehicleInterface|GetOwnerLocation (Variables|Default|GetVehicleInterface)) (select (Utilities|IsArrayNonEmpty (Vehicle|GetAllPassengers (Variables|Default|GetVehicleInterface))) (Variables|/Script/Engine.Pawn|GetPlayerState (Utilities|Array|Get(acopy) (Vehicle|GetAllPassengers (Variables|Default|GetVehicleInterface)))) 0))
                (CallFunction|VehicleTeleportEvent)
                (Development|PrintString "
============
SUCCESS
is a vehicle
============
"))
              (:CastFailed
                (bind _asfort_aipawn (Utilities|Casting|CastToFortAIPawn (Variables|RiftPortal|GetOverlapActor)))
                (Utilities|IsValid (.AttachedPlayerPawn (CallFunction|IsPlayerAttachedtoAIPawn _asfort_aipawn))
                  (:Is%20Valid
                    (CallFunction|TryRiftTeleport (.AttachedPlayerPawn (CallFunction|IsPlayerAttachedtoAIPawn _asfort_aipawn))) ; Teleport the Rider first
                  )
                  (:Is%20Not%20Valid
                    _))
                (CallFunction|PlayTeleportFX (Transformation|GetActorLocation) (Variables|/Script/Engine.Pawn|GetPlayerState _asfort_aipawn))
                (Environment|ApplyGlobalEnvironmentGameplayEffecttoActor 0 "/Game/Athena/Items/Consumables/RiftItem/GE_RiftPortal_Teleported.GE_RiftPortal_Teleported_C")
                (AI|TeleportAIPawn _asfort_aipawn (Math|Vector|vector+vector (Variables|RiftPortal|GetTeleportLocation) (1e+02 1e+02 0.0)) "0, 0, 0" true)
                (Variables|/Script/FortniteGame.FortPawn|SetbIgnoreNextFallingDamage true _asfort_aipawn)
                (Ability|SendGameplayEventtoActor _asfort_aipawn "(TagName=\"Abilities.Generic.Rift.Trigger\")")))))))))

(event EventOverlappingOnBeginPlay
  (params (Actor :type Actor%20Object%20Reference))
  ;[1 1 1] Teleport Player and/or Vehicle
  (CallFunction|TryRiftTeleport Actor))

(event Custom|VehicleTeleportEvent
  (bind _vehicleinterface (Variables|Default|GetVehicleInterface))
  ;[1 1 1] Teleport Player and/or Vehicle
  ;;[1 1 1] Vehicle Teleport
  (if (Vehicle|CanTeleport _vehicleinterface)
    (bind _returnvalue (Vehicle|GetVehicleActor _vehicleinterface :!class /Script/FortniteGame.FortVehicleInterface))
    (CallFunction|TeleportVehicle _returnvalue (+ (Variables|RiftPortal|GetTeleportLocation) (Variables|Default|GetTeleportLocationOffset)))
    (Vehicle|SetIgnoreNextFallingDamage _vehicleinterface true)
    (CallFunction|IncrementTeleportLocationOffset (* (.BoxExtent (Collision|GetActorBounds _returnvalue true false)) "2.500000,2.500000,0.000000"))
    (Development|PrintString (Utilities|String|BuildString(Name) "" "
==========
INCREMENT
" (Utilities|String|StringToName (Utilities|GetDisplayName _returnvalue)) "
==========
"))))

(event Custom|VehicleInteractionFailed
  (params (FortPlayerPawnAthena :type Fort%20Player%20Pawn%20Athena%20Object%20Reference))
  ;[1 1 1] Teleport Player and/or Vehicle
  ;;[1 1 1] For handling local player interacting with a vehicle before server confirmation
  (if (Utilities|Array|ContainsItem (Collision|GetOverlappingActors (Variables|Default|GetOverlapCapsule) 0 :!class /Script/Engine.PrimitiveComponent) FortPlayerPawnAthena)
    (CallFunction|TryRiftTeleport FortPlayerPawnAthena)))

(event EventNotifyTeleportedVehicle
  (params (VehicleActor :type Actor%20Object%20Reference))
  (bind _vehicleinterface (Variables|Default|GetVehicleInterface))
  (bind _returnvalue (Vehicle|GetDriver _vehicleinterface))
  (bind _playerpawn (Variables|RiftPortal|GetPlayerPawn))
  ;[1 1 1] Teleport Player and/or Vehicle
  (bind _asfort_vehicle_interface (Utilities|Casting|CastToFortVehicleInterface VehicleActor))
  (bind _output_get (Variables|Default|SetVehicleInterface _asfort_vehicle_interface))
  ;;[0.185 0.3042 1] Remove the passengers from the vehicle if there are any.
  (bind _output_get_2 (Variables|RiftPortal|SetPassengerArray (Vehicle|GetAllPassengers _output_get)))
  (for _array_element _output_get_2
    (RemoteControlPawn|ForceDetachfromRemoteControlledPawn _array_element)
    (Ability|SendGameplayEventtoActor _array_element "(TagName=\"Abilities.Generic.Rift.Trigger\")"))
  ;[1 1 1] SERVER: Check to Award Single Use Quest Progression if not in Creative Mode
  (if (not (Variables|QuestStuff|GetCreativeMode))
    (Networking|SwitchHasAuthority
      (:Authority
        (Utilities|IsValid _playerpawn
          (:Is%20Valid
            (CallFunction|UseRiftQuestUpdate (Pawn|GetController _playerpawn)))
          (:Is%20Not%20Valid
            (Utilities|IsValid (Utilities|ToObject(Interface) _vehicleinterface)
              (:Is%20Valid
                (Utilities|IsValid _returnvalue
                  (:Is%20Valid
                    (CallFunction|UseRiftQuestUpdate (Pawn|GetController _returnvalue)))
                  (:Is%20Not%20Valid
                    _))
                (for _array_element_2 (Vehicle|GetAllPassengers _vehicleinterface)
                  (Utilities|IsValid _array_element_2
                    (:Is%20Valid
                      (CallFunction|UseRiftQuestUpdate (Pawn|GetController _array_element_2)))))))))))))

(event OnComponentBeginOverlap(Capsule)
  (params (OverlappedComponent :type Primitive%20Component%20Object%20Reference) (OtherActor :type Actor%20Object%20Reference) (OtherComp :type Primitive%20Component%20Object%20Reference) (OtherBodyIndex :type Integer) (bFromSweep :type Boolean) (SweepResult :type Hit%20Result%20Structure%20%28by%20ref%29))
  (bind _sm_rift (Variables|Default|GetSM_Rift))
  ;[1 1 1] On Overlap - Toggle Post Process
  (Utilities|Casting|CastToFortPlayerPawn OtherActor)
  (CallFunction|PlayExpandorCollapseAudio true)
  ;;[1 1 1] Widen rift if player is close
  (bind (_direction _open) (Timeline|VectorScale Play)
    (:Update
      (Rendering|Material|SetScalarParameterValueonMaterials _sm_rift "Open" (Math|Float|SelectFloat 0.0 _open (Components|Timeline|IsPlaying (Variables|BGA_RiftPortal_Item_Athena_Delay|GetFragmentMovement))))
      (Transformation|SetRelativeRotation _sm_rift (Math|Rotator|MakeRotator (* _open -2e+01)) false false)
      (Effects|Components|ParticleSystem|SetFloatParameter (Variables|Default|GetNS_Rift_Idle_Loop) "Open" _open))))

(event OnComponentEndOverlap(Capsule)
  (params (OverlappedComponent :type Primitive%20Component%20Object%20Reference) (OtherActor :type Actor%20Object%20Reference) (OtherComp :type Primitive%20Component%20Object%20Reference) (OtherBodyIndex :type Integer))
  (bind _sm_rift (Variables|Default|GetSM_Rift))
  ;[1 1 1] On Overlap - Toggle Post Process
  (bind _asfort_player_pawn_athena (Utilities|Casting|CastToFortPlayerPawnAthena OtherActor)
    (:then
      (EventDispatchers|UnbindEventfromOnInteractionFailedCallback _asfort_player_pawn_athena (EventDispatchers|CreateEvent :!event Custom|VehicleInteractionFailed)))
    (:CastFailed
      _))
  (if (<= (Utilities|Array|Length (Collision|GetOverlappingActors OtherActor "/Script/FortniteGame.FortPlayerPawn")) 0)
    (CallFunction|PlayExpandorCollapseAudio false)
    ;;[1 1 1] Widen rift if player is close
    (Timeline|VectorScale Reverse)))

(event Custom|DestroyRift
  (bind _audio_looping (Variables|Sound|GetAudio_Looping))
  ;[1 1 1] On Death
  (Variables|RiftPortal|SetPortalOpen false)
  (Utilities|FlowControl|Sequence
    (:then_0
      (Networking|CanExecuteCosmeticEvents
        (:True
          (Development|SetHiddeninGame (Variables|Default|GetSM_Rift) true false)
          (Audio|PlaySoundatLocation (Variables|Sound|GetRiftDestroyed) (Transformation|GetActorLocation)) ; Play Deactivate Sound
          (bind (_direction _pitchcurve) (Timeline|Audio%20RiftDestroyed Play)
            (:Update
              (Audio|Components|Audio|SetPitchMultiplier _audio_looping _pitchcurve))
            (:Finished
              (Audio|Components|Audio|FadeOut _audio_looping 1 0.0 "Linear" :!class /Script/Engine.AudioComponent))))))
    (:then_1
      (Utilities|FlowControl|Delay (+ (ScalableFloat|GetValueatLevel (Variables|RiftPortal|GetHangtime)) 1.0))
      (Actor|DestroyActor))))

(event Custom|IncrementTeleportLocationOffset
  (params (IncrementValue :type Vector))
  ;[0 0 0] CUSTOM EVENT -> IncrementTeleportLocationOffset
  ; Each time an actor teleports we increment this offset so they don't overlap. This is for when things teleport together quickly.
  (Variables|Default|SetTeleportLocationOffset (+ (Variables|Default|GetTeleportLocationOffset) IncrementValue))
  (Utilities|Time|SetTimerbyEvent (delegate Custom|ResetTeleportLocationOffset) 1.0))

(event Custom|ResetTeleportLocationOffset
  ;[0 0 0] CUSTOM EVENT -> IncrementTeleportLocationOffset
  ; Each time an actor teleports we increment this offset so they don't overlap. This is for when things teleport together quickly.
  (Variables|Default|SetTeleportLocationOffset "0, 0, 0")
  (Utilities|Set|Clear (Variables|Default|GetBeganOverlapList)))

(event Custom|TeleportPlayerAndSendEvent
  (params (PlayerPawn :type Fort%20Player%20Pawn%20Object%20Reference))
  (bind (_origin _boxextent) (Collision|GetActorBounds PlayerPawn true false))
  (bind (_actorrotation_roll _actorrotation_pitch _actorrotation_yaw) (Math|Rotator|BreakRotator (Variables|GetActorRotation_Roll)))
  (Transform|Actor|TeleportPlayerPawn PlayerPawn (+ (Variables|RiftPortal|GetTeleportLocation) (Variables|Default|GetTeleportLocationOffset)) (Math|Rotator|MakeRotator 0.0 0.0 _actorrotation_yaw) true false)
  (Ability|SendGameplayEventtoActor PlayerPawn "(TagName=\"Abilities.Generic.Rift.Trigger\")")
  (CallFunction|IncrementTeleportLocationOffset (* _boxextent "6.000000,6.000000,0.000000")))

(fn CancelPlayerPawnAbilities
  (bind _playerpawn (Variables|RiftPortal|GetPlayerPawn)
    (:then
      (Utilities|IsValid (Ability|GetFortAbilitySystemComponent _playerpawn)
        (:Is%20Valid
          (Tags|CancelAbilitiesWithTags (Ability|GetFortAbilitySystemComponent _playerpawn) (Variables|Default|GetCancelAbilitiesTags)))))
    (:else)))

(fn CheckForRiftLocUpdate
  (params (InController :type Controller%20Object%20Reference)))

(fn EjectFromTurret
  (params (PlayerPawn :type Fort%20Player%20Pawn%20Object%20Reference))
  (bind _returnvalue (Vehicle|GetVehicleInterface PlayerPawn))
  (if (Vehicle|IsinVehicle PlayerPawn)
    (Utilities|IsValid (Utilities|ToObject(Interface) _returnvalue)
      (:Is%20Valid
        (if (not (Vehicle|CanTeleport _returnvalue))
          (GameplayEffects|RemoveActiveEffectswithGrantedTags (Ability|GetAbilitySystemComponent PlayerPawn) "(GameplayTags=((TagName=\"Gameplay.Action.Player.AthenaInVehicle\")))") ; If vehicle isn't supposed to teleport then kick the player from it
        ))))
  (return))

(fn GetAffiliationsToShowFor
  (return (Utilities|Array|MakeArray "Friendly" "Neutral" "Hostile")))

(fn InitTeleportLocation
  (Variables|SetTeleportLocation_X (Math|Vector|MakeVector (.x (Transformation|GetActorLocation)) (.y (Transformation|GetActorLocation)) (ScalableFloat|GetValueatLevel (Variables|RiftPortal|GetTeleportHeight)))))

(fn IsFlying
  (bind _asfort_movement_comp_character (Utilities|Casting|CastToFortMovementComp_Character (Pawn|GetMovementComponent (Variables|RiftPortal|GetPlayerPawn)))
    (:then
      (return (Utilities|Enum|Equal(Enum) (Variables|/Script/Engine.CharacterMovementComponent|GetMovementMode _asfort_movement_comp_character) "MOVE_Flying")))
    (:CastFailed
      (return false))))

(fn IsPlayerAttachedToAIPawn
  (params (AIPawn :type Fort%20AIPawn%20Object%20Reference))
  (for _array_element (Actor|GetAttachedActors AIPawn true false)
    (bind _asfort_player_pawn (Utilities|Casting|CastToFortPlayerPawn _array_element))
    (return true _asfort_player_pawn))
  (return false 0))

(fn OnRep_PlayerPawn)

(fn OnRep_TeleportLocation)

(fn OnRep_Vehicle)

(fn PlayExpandOrCollapseAudio
  (params (Expand :type Boolean))
  (if (Variables|RiftPortal|GetPortalOpen)
    (Audio|PlaySoundatLocation (select Expand "/Game/Sounds/Fort_GamePlay_Sounds/Player/RiftPortal/RiftPortal_Expand_Cue.RiftPortal_Expand_Cue" "/Game/Sounds/Fort_GamePlay_Sounds/Player/RiftPortal/RiftPortal_Collapse_Cue.RiftPortal_Collapse_Cue") (Transformation|GetWorldLocation (Variables|Default|GetSM_Rift)))))

(fn ShouldShowSoundIndicator
  (params (PlayerController :type Fort%20Player%20Controller%20Object%20Reference))
  (bind _returnvalue (Game|Player|GetPlayerPawn PlayerController))
  (Utilities|IsValid (Variables|/Script/FortniteGame.FortSoundIndicatorComponent|GetRequestingPlayerState (Variables|Default|GetSoundIndicator))
    (:Is%20Valid
      (bind _vehicle_occupied (Utilities|IsinVehicle _returnvalue)
        (:true
          (bind _asfort_player_pawn (Utilities|Casting|CastToFortPlayerPawn (PlayerState|GetPawn (Variables|/Script/FortniteGame.FortSoundIndicatorComponent|GetRequestingPlayerState (Variables|Default|GetSoundIndicator))))
            (:then
              (return (== (Class|FortVehicleInterface|FindSeatIndex _vehicle_occupied _asfort_player_pawn) -1)))
            (:CastFailed
              _)))
        (:false
          _)))
    (:Is%20Not%20Valid
      _))
  (return (!= (Variables|/Script/Engine.Pawn|GetPlayerState _returnvalue) (Variables|/Script/FortniteGame.FortSoundIndicatorComponent|GetRequestingPlayerState (Variables|Default|GetSoundIndicator)))))

(fn ConstructionScript)

(fn UseRiftQuestUpdate
  (params (InController :type Controller%20Object%20Reference))
  (Utilities|IsValid InController
    (:Is%20Valid
      (bind _asfort_player_controller (Utilities|Casting|CastToFortPlayerController InController))
      (bind (_questisvalid _questcompleted) (Default|AthenaCheckQuestandObjectiveCompletion _asfort_player_controller (Variables|QuestStuff|GetQuestItem) (Variables|QuestStuff|GetObjectiveBackendName)))
      (if (and _questisvalid (not _questcompleted))
        (Class|AthenaQuestFunctionLibrary|AthenaGiveQuestUpdatetoPlayers :InQuestItemRef (Variables|QuestStuff|GetQuestItem) :ObjectiveBackendName (Variables|QuestStuff|GetObjectiveBackendName) :PlayerControllersForUpdate (Utilities|Array|MakeArray _asfort_player_controller))))))
